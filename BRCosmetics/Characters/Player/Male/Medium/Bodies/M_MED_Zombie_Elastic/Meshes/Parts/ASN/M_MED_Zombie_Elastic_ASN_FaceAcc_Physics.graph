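# Physics asset for the face accessory mesh of the medium male Zombie Elastic ASN character part, covering four dynamic hair bones and the head.
# Each bone carries a single capsule body, with a large one on the head and smaller ones on the hair bones.
# All four hair bones are constrained directly to the head rather than chained, with per-bone swing limits and a tight twist limit, so each swings independently with little twist.
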PhysicsAsset /BRCosmetics/Characters/Player/Male/Medium/Bodies/M_MED_Zombie_Elastic/Meshes/Parts/ASN/M_MED_Zombie_Elastic_ASN_FaceAcc_Physics
bodies {
 dyn_hair_01 [{shapeType=Capsule, length=1.9391, radius=1.7138, center=<1 0 0>, rotation=<rot 34.8034 0 0>}]
 dyn_hair_02 [{shapeType=Capsule, length=0.5389, radius=1.1007, shapeName=dyn_hair_02_capsule, center=<1 0 0>, rotation=<rot 34.803 0 0>}]
 dyn_hair_03 [{shapeType=Capsule, length=0.5389, radius=1.1007, shapeName=dyn_hair_03_capsule, center=<1 0 0>, rotation=<rot 34.803 0 0>}]
 dyn_hair_04 [{shapeType=Capsule, length=0.5389, radius=1.1007, shapeName=dyn_hair_04_capsule, center=<1 0 0>, rotation=<rot 34.803 0 0>}]
 head [{shapeType=Capsule, length=6.1642, radius=6.0478, center=<5.1435 2.3809 -0.1656>, rotation=<rot 0 0 90>}]
}
constraints [
 {bone1Name=dyn_hair_01, bone2Name=head, swing1LimitDegrees=45, swing2LimitDegrees=45, twistLimitDegrees=3}
 {bone1Name=dyn_hair_04, bone2Name=head, swing1LimitDegrees=35, swing2LimitDegrees=35, twistLimitDegrees=3}
 {bone1Name=dyn_hair_03, bone2Name=head, swing1LimitDegrees=25, swing2LimitDegrees=25, twistLimitDegrees=3}
 {bone1Name=dyn_hair_02, bone2Name=head, swing1LimitDegrees=45, swing2LimitDegrees=45, twistLimitDegrees=3}
]
